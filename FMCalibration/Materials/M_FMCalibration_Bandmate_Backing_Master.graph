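# M_FMCalibration_Bandmate_Backing_Master is a translucent UI master material for a bandmate backing panel with soft-border or glow variants and an overdrive look.
# An aspect-corrected rounded-box SDF drives soft-border, anti-aliased and sine-animated glow masks, selected for opacity by IsGlow.
# Color is either a simple animatable blend or a rotated gradient that fades into overdrive triangle colors, with optional intensity oscillation.
# A panned, time-remapped texture builds the triangle pattern mask.
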Material /FMCalibration/Materials/M_FMCalibration_Bandmate_Backing_Master
graph {
 properties {
  bUsedWithStaticMesh=false
  blendMode=BLEND_Translucent
  materialDomain=MD_UI
  pixelDepthOffsetMode=PDOM_Legacy
  usedShadingModels=Default Lit
 }
 nodes [
  Constant() {r=1}
  n1 = VectorParameter() {parameterName=TriangleFillColor_OD, defaultValue=<rgba 1 0 0 1>}
  n2 = TextureCoordinate()
  n3 = LinearInterpolate(
   A=LinearInterpolate(
    A=VectorParameter() {parameterName=OverdriveTriColor_A, defaultValue=<rgba 0 1 1 1>}.RGB
    B=VectorParameter() {parameterName=OverdriveTriColor_B, defaultValue=<rgba 1 1 0 1>}.RGB
    Alpha=NamedRerouteUsage()
   )
   B=@n1.RGB
   Alpha=ComponentMask(None=SmoothStep(Value=@n2) {constMax=0.9}) {r=true}
  )
  LinearInterpolate(A=@n1.RGB, B=@n3, Alpha=NamedRerouteUsage())
  n4 = Multiply(
   A=MaterialFunctionCall() {materialFunction=&ScreenAlignedUVs}."X 100%, Y 100%"
   B=ScalarParameter() {defaultValue=3, parameterName=Tile}
  ) {constB=2}
  MaterialFunctionCall(UV=@n4) {materialFunction=&MF_UI_UVScaling}
  MaterialFunctionCall() {materialFunction=&MF_UI_UVScaling}
  n5 = ScalarParameter() {defaultValue=0.125, parameterName=Gradient_Rotation}
  n6 = NamedRerouteDeclaration(None=ComponentMask(None=MaterialFunctionCall(UVs=@n2, "Rotation Angle (0-1)"=@n5) {materialFunction=&CustomRotator}.Rotated Values) {r=true}) {name=GradientMask, nodeColor=<rgba 1 0.6328 0 1>}
  n7 = Multiply(A=NamedRerouteUsage(), B=OneMinus(None=@n6))
  n8 = NamedRerouteUsage()
  n9 = OneMinus(None=ScalarParameter() {defaultValue=0.25, parameterName=SoftBorderHardness})
  n10 = ScalarParameter() {defaultValue=0.05, parameterName=Threshold}
  n11 = MaterialFunctionCall() {materialFunction=&GetUserInterfaceUV}
  n12 = MaterialFunctionCall(Float2=@n11.Pixel Size) {materialFunction=&BreakOutFloat2Components}
  n13 = NamedRerouteDeclaration(None=@n12.G) {name=PixelSize.Y, nodeColor=<rgba 0 0 0 1>}
  n14 = Max(A=NamedRerouteDeclaration(None=@n12.R) {name=PixelSize.X, nodeColor=<rgba 0 0 0 1>}) {constB=0.001}
  n15 = Multiply(A=@n10, B=Multiply(A=Divide(A=@n13, B=@n14)))
  n16 = MaterialFunctionCall(
   Size=Subtract(B=AppendVector(A=@n15, B=@n10))
   CornerRoundness=ScalarParameter() {defaultValue=0.7, parameterName=CornerRadius}
   KeepAspectRatio=StaticBool() {value=true}
  ) {materialFunction=&MF_UI_Box}
  n17 = NamedRerouteDeclaration(None=SmoothStep(Max=Multiply(A=@n9) {constB=-1}, Value=@n16.SDF) {constMax=-0.25}) {name=SoftBorder, nodeColor=<rgba 1 0.7031 0 1>}
  n18 = ScalarParameter() {defaultValue=1, parameterName=SoftBorder}
  Multiply(
   A=StaticSwitchParameter(True=@n7, False=LinearInterpolate(A=@n8, B=@n17, Alpha=@n18)) {parameterName=IsGlow}
   B=ScalarParameter() {defaultValue=0.3, parameterName=Alpha_Final}
  )
  n19 = ScalarParameter() {defaultValue=0.05, parameterName=Threshold}
  NamedRerouteDeclaration(None=Saturate(None=MaterialFunctionCall(DistanceField=@n16.SDF, Threshold=@n19) {materialFunction=&MF_UI_AntiAliasing}.Inner)) {name=Box Mask, nodeColor=<rgba 0 0 0 1>}
  n20 = MaterialFunctionCall(Speed=ScalarParameter() {defaultValue=0.5}) {materialFunction=&DebugTimeSine}
  n21 = Constant2Vector() {g=-0.2, r=-0.1}
  n22 = TextureSample(UVs=NamedRerouteDeclaration(None=Panner(Coordinate=@n4, Speed=@n21)) {name=PanningUVs, nodeColor=<rgba 1 0.8672 0 1>}) {texture=/FMCalibration/Textures/T_FMCalibration_Track_PeakState}
  n23 = SmoothStep(Value=@n22.R) {constMax=0.25, constMin=0.1}
  n24 = SmoothStep(Value=@n22.R) {constMax=0.25, constMin=0.5}
  n25 = OneMinus(None=@n24)
  n26 = Multiply(A=@n20.Result, B=LinearInterpolate(A=@n23, B=@n25, Alpha=@n20.Result))
  n27 = SmoothStep(Value=@n22.R)
  n28 = OneMinus(None=@n27)
  n29 = Multiply(A=@n20.Result, B=LinearInterpolate(A=@n24, B=OneMinus(None=@n23), Alpha=LinearInterpolate(A=@n28, B=@n27, Alpha=@n25)))
  NamedRerouteDeclaration(None=Add(A=LinearInterpolate(A=@n26, B=@n29, Alpha=@n20.Result), B=@n27)) {name=TrianglePan, nodeColor=<rgba 0 1 0.0625 1>}
  NamedRerouteDeclaration(None=@n11.9-Slice UV) {name=9sliceuv, nodeColor=<rgba 0 0 0 1>}
  NamedRerouteDeclaration(None=@n11.Normalized UV) {name=NormalizedUV, nodeColor=<rgba 0 0 0 1>}
  n30 = SmoothStep(Value=@n16.SDF) {constMax=-0.5}
  n31 = TextureCoordinate()
  n32 = Constant() {r=-2}
  n33 = ComponentMask(None=Sine(None=Panner(Coordinate=@n31, Speed=@n32)) {period=2}) {r=true}
  n34 = LinearInterpolate(Alpha=Saturate(None=@n33)) {constA=0.75}
  n35 = LinearInterpolate(A=@n33, Alpha=OneMinus(None=NamedRerouteUsage()))
  n36 = Multiply(A=@n30, B=Saturate(None=Multiply(A=@n34, B=@n35)))
  n37 = ScalarParameter() {defaultValue=1, parameterName=SineWaveAmount}
  NamedRerouteDeclaration(None=MaterialFunctionCall("Value (0-1)"=LinearInterpolate(A=@n30, B=@n36, Alpha=@n37)) {materialFunction=&MF_UI_EaseCurves}.EaseInOutCubic) {name=GlowMask, nodeColor=<rgba 0 0 0 1>}
  NamedRerouteUsage()
  NamedRerouteUsage()
  OneMinus(None=@n22.R)
  SmoothStep(Value=@n22.R) {constMax=0.9, constMin=0.1}
  n38 = VectorParameter() {parameterName=Gradient_Color_A, defaultValue=<rgba 0 0 0 1>}
  n39 = VectorParameter() {parameterName=Gradient_Color_B, defaultValue=<rgba 0 0 0 1>}
  n40 = StaticSwitchParameter(True=@n7, False=@n6) {parameterName=IsGlow}
  n41 = ScalarParameter() {parameterName=OverdriveActive}
  n42 = ScalarParameter() {defaultValue=1, parameterName=OD_ColorIntensity}
  n43 = ScalarParameter() {defaultValue=1, parameterName=ODIntentsitySpeed}
  n44 = LinearInterpolate(
   A=@n42
   B=LinearInterpolate(
    A=Add(A=@n42) {constB=0.25}
    B=Add(A=@n42) {constB=-0.25}
    Alpha=MaterialFunctionCall(Speed=@n43) {materialFunction=&DebugTimeSine}.Result
   )
   Alpha=ScalarParameter() {parameterName=ODIntensityOscillate}
  )
  StaticSwitchParameter(
   True=LinearInterpolate(A=LinearInterpolate(A=@n38.RGB, B=@n39.RGB, Alpha=@n40), B=@n3, Alpha=Multiply(A=@n41, B=@n44))
   False=LinearInterpolate(
    A=VectorParameter() {parameterName=Color, defaultValue=<rgba 0 0 1 1>}.RGB
    B=VectorParameter() {parameterName=Color_AnimatedTo, defaultValue=<rgba 0 0 1 1>}.RGB
    Alpha=ScalarParameter() {parameterName=Anim_SimpleColor}
   )
  ) {defaultValue=true, parameterName=Gradient}
 ]
}
