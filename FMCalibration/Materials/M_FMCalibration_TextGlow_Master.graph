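# Translucent UI master material for a pulsing radial glow, named for calibration text in FMCalibration.
# A time sine at SineSpeed oscillates the radius of an exponential radial gradient between start and stop values.
# The gradient blends two color pairs, with IsMax interpolating from the normal pair to the Max pair.
# It also forms a glow mask, scaled by an Alpha-driven intensity range, shaped by the TaperGlow exponent and saturated.
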Material /FMCalibration/Materials/M_FMCalibration_TextGlow_Master
graph {
 properties {
  bUsedWithStaticMesh=false
  blendMode=BLEND_Translucent
  materialDomain=MD_UI
  pixelDepthOffsetMode=PDOM_Legacy
  usedShadingModels=Default Lit
 }
 nodes [
  n1 = ScalarParameter() {defaultValue=0.5, parameterName=SineSpeed}
  n2 = MaterialFunctionCall(
   Radius=LinearInterpolate(
    A=ScalarParameter() {defaultValue=0.4, parameterName=Radius_Start}
    B=ScalarParameter() {defaultValue=0.5, parameterName=Radius_Stop}
    Alpha=MaterialFunctionCall(Speed=@n1) {materialFunction=&DebugTimeSine}.Result
   ) {constA=0.4, constB=0.5}
   Density=ScalarParameter() {defaultValue=2.333, parameterName=Density}
  ) {materialFunction=&RadialGradientExponential}
  LinearInterpolate(
   A=LinearInterpolate(
    A=VectorParameter() {parameterName=Color_A, defaultValue=<rgba 0 1 0 1>}.RGB
    B=VectorParameter() {parameterName=Color_B, defaultValue=<rgba 0.0612 0.8714 0.4735 1>}.RGB
    Alpha=@n2.RadialGradientExponential
   )
   B=LinearInterpolate(
    A=VectorParameter() {parameterName=Color_A_Max, defaultValue=<rgba 0 1 0 1>}.RGB
    B=VectorParameter() {parameterName=Color_B_Max, defaultValue=<rgba 0.0612 0.8714 0.4735 1>}.RGB
    Alpha=@n2.RadialGradientExponential
   )
   Alpha=ScalarParameter() {parameterName=IsMax}
  )
  n3 = LinearInterpolate(
   A=ScalarParameter() {defaultValue=0.25, parameterName=Alpha_Start}
   B=ScalarParameter() {defaultValue=0.5, parameterName=Alpha_Stop}
   Alpha=ScalarParameter() {parameterName=Alpha}
  ) {constA=1.5953}
  Saturate(None=Power(Base=Multiply(A=@n2.RadialGradientExponential, B=@n3), Exp=ScalarParameter() {defaultValue=1, parameterName=TaperGlow}))
 ]
}
