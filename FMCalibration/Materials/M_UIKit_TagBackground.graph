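# M_UIKit_TagBackground is a translucent UI material that draws a tag background as a skewed, rounded rectangle with optional outline.
# MF_UI_CoreUI_RoundedMask builds body and outline distance fields, nudged by -0.01 to keep skewed edges off the bounds, then anti-aliased.
# A static switch picks flat or linear-gradient fill, blended with the outline color where floored hasOutline allows.
# The body mask times an Opacity parameter gives the alpha.
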Material /FMCalibration/Materials/M_UIKit_TagBackground
graph {
 properties {
  bUsedWithStaticMesh=false
  blendMode=BLEND_Translucent
  materialDomain=MD_UI
  pixelDepthOffsetMode=PDOM_Legacy
  usedShadingModels=Default Lit
 }
 nodes [
  n1 = VectorParameter() {parameterName=Color_Block, defaultValue=<rgba 1 1 1 1>}
  n2 = VectorParameter() {parameterName=Color Block Accent, defaultValue=<rgba 0.1167 0.9155 1 1>}
  n3 = MaterialFunctionCall(
   GradientStart=ScalarParameter() {parameterName=Block Gradient Start, sortPriority=10}
   GradientEnd=ScalarParameter() {defaultValue=1, parameterName=Block Gradient End, sortPriority=11}
   RotationAngle=ScalarParameter() {parameterName=Block Gradient Rotation, sortPriority=12}
   GradientOffset=ScalarParameter() {parameterName=Block Gradient Offset, sortPriority=13}
  ) {materialFunction=&MF_UI_LinearGradient}
  n4 = ScalarParameter() {parameterName=hasOutline}
  n5 = MaterialFunctionCall(
   Size=MaterialFunctionCall(
    X=ScalarParameter() {defaultValue=0.85, group=Layout, parameterName=ScaleX, sortPriority=1}
    Y=ScalarParameter() {defaultValue=0.85, group=Layout, parameterName=ScaleY, sortPriority=1}
   ) {materialFunction=&MakeFloat2}.Result
   Corner Radius=MaterialFunctionCall(
    X=ScalarParameter() {defaultValue=0.25, group=Layout, parameterName=TopLeft, sortPriority=10}
    Y=ScalarParameter() {defaultValue=0.25, group=Layout, parameterName=BottomLeft, sortPriority=13}
    Z=ScalarParameter() {defaultValue=0.25, group=Layout, parameterName=TopRight, sortPriority=11}
    A=ScalarParameter() {defaultValue=0.25, group=Layout, parameterName=BottomRight, sortPriority=12}
   ) {materialFunction=&MakeFloat4}.Result
   Radius in Pixels=ScalarParameter() {group=Layout, parameterName=Radius in Pixels, sortPriority=9}
   Skew Angle=ScalarParameter() {defaultValue=10, group=Layout, parameterName=SkewAngle, sortPriority=4}
   Outline Thickness=Multiply(A=ScalarParameter() {defaultValue=4, parameterName=OutlineThickness}, B=Floor(None=@n4))
  ) {materialFunction=&MF_UI_CoreUI_RoundedMask}
  n6 = Add(A=@n5.Outline SDF) {constB=-0.01, desc=-.01 so when skewed it'll not be at the edge if scale is set to 1}
  n7 = ScalarParameter() {parameterName=hasOutline}
  LinearInterpolate(
   A=StaticSwitchParameter(True=LinearInterpolate(A=@n1.RGB, B=@n2.RGB, Alpha=@n3.Result), False=@n1.RGB) {parameterName=Block Gradient}
   B=VectorParameter() {group=Color, parameterName=ColorOutline, defaultValue=<rgba 1 1 0 1>}.RGB
   Alpha=Multiply(
    A=MaterialFunctionCall(DistanceField=@n6) {materialFunction=&MF_UI_AntiAliasing}.Inner
    B=Floor(None=@n7)
   )
  ) {constA=0.09}
  Multiply(
   A=MaterialFunctionCall(DistanceField=Add(A=@n5.SDF) {constB=-0.01}) {materialFunction=&MF_UI_AntiAliasing}.Inner
   B=ScalarParameter() {defaultValue=1, parameterName=Opacity, sliderMax=1}
  )
 ]
}
