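# Translucent UI material for a calibration display, drawing a bus icon and a rounded vertical line marker, each positioned horizontally by parameters.
# The icon comes from an MSDF texture, scaled and antialiased.
# The line is an MF_UI_Box fill and stroke, shaped by a sine-based vertical mask and tinted with a noise-modulated three-color blue gradient.
# A static Debug switch replaces both position parameters with time-driven animation.
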Material /FMCalibration/Materials/M_UI_Calibration_BusIcon
graph {
 properties {
  bAutomaticallySetUsageInEditor=false
  bUsedWithStaticMesh=false
  blendMode=BLEND_Translucent
  materialDomain=MD_UI
  usedShadingModels=Default Lit
 }
 nodes [
  LinearInterpolate(A=NamedRerouteUsage(), Alpha=NamedRerouteUsage())
  LinearInterpolate(A=Constant() {r=1}, B=NamedRerouteUsage(), Alpha=NamedRerouteUsage())
  n1 = NamedRerouteUsage()
  Multiply(A=@n1) {constB=5}
  n2 = AppendVector(A=Constant() {r=0.8}, B=Constant() {r=1})
  n3 = ScalarParameter() {defaultValue=0.35, parameterName=Texture Scale}
  n4 = Time()
  n5 = StaticSwitchParameter(True=Frac(None=Multiply(A=@n4) {constB=0.5}), False=ScalarParameter() {defaultValue=0.2, parameterName=BusPosition}) {parameterName=Debug}
  n6 = Add(A=Constant() {r=0.5}, B=ScalarParameter() {defaultValue=0.06, parameterName=BusOffset}) {constB=0}
  n7 = Multiply(A=@n6) {constB=-1}
  n8 = Add(A=TextureCoordinate(), B=AppendVector(A=LinearInterpolate(A=@n6, B=@n7, Alpha=@n5) {constA=0.5, constB=-0.5}, B=Constant()))
  n9 = TextureSample(UVs=MaterialFunctionCall(
    UVs=MaterialFunctionCall(UV=@n8) {materialFunction=&MF_UI_UVScaling}.ScaleY
    Texture Scale=Multiply(A=@n2, B=@n3)
   ) {materialFunction=&ScaleUVsByCenter}.UVs) {samplerSource=SSM_Clamp_WorldGroupSettings, samplerType=SAMPLERTYPE_LinearColor, texture=/FMCalibration/Textures/T_UI_CalibrationBus_IconSDF}
  NamedRerouteDeclaration(None=MaterialFunctionCall(
    DistanceField=MaterialFunctionCall!(MSDF RGB=@n9.RGB) {materialFunction=/FortUILibrary/MaterialFunctions/Shapes/Modifiers/MF_UI_MSDFtoSDF}.SDF
    Threshold=Constant() {r=0.5}
   ) {materialFunction=&MF_UI_AntiAliasing}.Outer) {name=Bus_Icon, nodeColor=<rgba 0.1719 0 1 1>}
  n10 = MaterialFunctionCall(Speed=Constant() {r=0.23}) {materialFunction=&DebugTimeSine}
  n11 = Add(A=Constant() {r=0.5}, B=ScalarParameter() {defaultValue=-0.06, parameterName=LineOffset}) {constB=0}
  n12 = LinearInterpolate(
   A=@n11
   B=Multiply(A=@n11) {constB=-1}
   Alpha=StaticSwitchParameter(True=Frac(None=@n10.Result), False=ScalarParameter() {parameterName=LinePositiopn}) {parameterName=Debug}
  ) {constA=0.5, constB=-0.5}
  n13 = Constant()
  n14 = ScalarParameter() {defaultValue=0.1, parameterName=LineSizeX}
  n15 = MaterialFunctionCall(
   UVs=Add(A=TextureCoordinate(), B=AppendVector(A=@n12, B=@n13))
   Size=AppendVector(A=@n14, B=Constant() {r=0.9})
   CornerRoundness=@n14
   OutlineThickness=ScalarParameter() {defaultValue=0.03, parameterName=OutlineThicknes}
   OutlineAlignment=Constant()
   KeepAspectRatio=StaticBool() {value=true}
  ) {materialFunction=&MF_UI_Box}
  n16 = TextureCoordinate()
  n17 = OneMinus(None=ComponentMask(None=Sine(None=@n16) {period=2}) {g=true})
  n18 = LinearInterpolate(
   B=@n15.Fill
   Alpha=NamedRerouteDeclaration(None=SmoothStep(Value=@n17) {constMax=0.8, constMin=-0.5}) {name=InnerMask, nodeColor=<rgba 0 0 0 1>}
  )
  n19 = Multiply(A=NamedRerouteDeclaration(None=@n15.Outline) {name=Stroke, nodeColor=<rgba 0 0.7656 1 1>}) {constB=0.8}
  NamedRerouteDeclaration(None=Saturate(None=LinearInterpolate(A=@n18, Alpha=@n19))) {name=BarMask, nodeColor=<rgba 0.1484 1 0 1>}
  n20 = VectorNoise()
  n21 = MaterialFunctionCall(
   A=Constant3Vector() {constant=<rgba 0 0.0776 0.3185 1>}
   B=Constant3Vector() {constant=<rgba 0 0.5565 1 1>}
   C=Constant3Vector() {constant=<rgba 0.1094 0.9343 1 1>}
   Alpha=LinearInterpolate(A=@n1, Alpha=Multiply(A=@n20) {constB=0.4})
  ) {materialFunction=&Lerp_3Color}
  NamedRerouteDeclaration(None=LinearInterpolate(A=Saturate(None=@n21.Result), Alpha=NamedRerouteUsage()) {constA=1}) {name=BarColour, nodeColor=<rgba 0 1 0.2734 1>}
 ]
}
