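# Translucent UI material for the calibration point line, drawing a rounded, outlined box with a soft inner fill.
# MF_UI_Box supplies fill and outline, with a smoothstepped sine of vertical UV shaping the fill before the stroke blends in as a mask.
# A three-colour blue-to-mint gradient, driven partly by vector noise, is lerped toward white to form the colour.
# Named reroutes carry colour, inner mask and stroke.
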Material /FMCalibration/Materials/M_UI_Calibration_PointLine
graph {
 properties {
  bAutomaticallySetUsageInEditor=false
  bUsedWithStaticMesh=false
  blendMode=BLEND_Translucent
  materialDomain=MD_UI
  usedShadingModels=Default Lit
 }
 nodes [
  n1 = VectorNoise()
  n2 = MaterialFunctionCall(
   A=Constant3Vector() {constant=<rgba 0 0.0776 0.3185 1>}
   B=Constant3Vector() {constant=<rgba 0 0.5565 1 1>}
   C=Constant3Vector() {constant=<rgba 0.5395 1 0.6724 1>}
   Alpha=LinearInterpolate(A=NamedRerouteUsage(), Alpha=Multiply(A=@n1) {constB=0.4})
  ) {materialFunction=&Lerp_3Color}
  NamedRerouteDeclaration(None=LinearInterpolate(A=Saturate(None=@n2.Result), Alpha=NamedRerouteUsage()) {constA=1}) {name=Colour, nodeColor=<rgba 0 1 0.2734 1>}
  NamedRerouteUsage()
  n3 = TextureCoordinate()
  n4 = MaterialFunctionCall(
   UVs=@n3
   Size=Constant() {r=0.9}
   CornerRoundness=Constant() {r=0.2}
   OutlineThickness=Constant() {r=0.05}
   OutlineAlignment=Constant() {r=0.8}
   KeepAspectRatio=StaticBool() {value=true}
  ) {materialFunction=&MF_UI_Box}
  n5 = ComponentMask(None=Sine(None=@n3) {period=2}) {g=true}
  n6 = NamedRerouteDeclaration(None=SmoothStep(Value=OneMinus(None=@n5)) {constMin=-0.3}) {name=InnerMask, nodeColor=<rgba 0 0 0 1>}
  n7 = NamedRerouteDeclaration(None=@n4.Outline) {name=Stroke, nodeColor=<rgba 0 0.7656 1 1>}
  Saturate(None=LinearInterpolate(A=LinearInterpolate(B=@n4.Fill, Alpha=@n6), Alpha=Multiply(A=@n7) {constB=0.8}))
 ]
}
