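# M_UI_FMCalibrationLIne is a UI-domain material in the FMCalibration plugin, named for a calibration line.
# Its graph contains no expressions, so it computes no color, opacity or animation of its own.
# Automatic usage flagging in the editor and static mesh usage are both turned off.
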Material /FMCalibration/Materials/M_UI_FMCalibrationLIne
graph {
 expressions {}
 properties {bAutomaticallySetUsageInEditor=false, bUsedWithStaticMesh=false, materialDomain=MD_UI, usedShadingModels=Default Lit}
}
