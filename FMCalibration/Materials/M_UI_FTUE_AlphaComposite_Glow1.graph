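# UI material with alpha-composite blending that draws an animated, noise-distorted radial glow, apparently for FTUE screens.
# Scrolling blurred noise perturbs the UVs of an exponential radial gradient, shifted by Y Position, while time-driven sines pulse its radius and density.
# The gradient blends Color_A to Color_B and, scaled by an Alpha_Anim-driven intensity and shaped by TaperGlow, forms a saturated mask that premultiplies the color.
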Material /FMCalibration/Materials/M_UI_FTUE_AlphaComposite_Glow1
graph {
 properties {
  bUsedWithStaticMesh=false
  blendMode=BLEND_AlphaComposite
  materialDomain=MD_UI
  pixelDepthOffsetMode=PDOM_Legacy
  usedShadingModels=Default Lit
 }
 nodes [
  Constant() {r=0.5}
  n1 = TextureCoordinate()
  n2 = AppendVector(A=Constant(), B=ScalarParameter() {defaultValue=-0.5, parameterName=Y Position})
  n3 = AppendVector(A=Constant() {r=1}, B=Constant() {r=2})
  n4 = Constant()
  n5 = Constant() {r=0.1}
  n6 = TextureSample(UVs=Panner(Coordinate=Multiply(A=@n1, B=@n3), Speed=AppendVector(A=@n4, B=@n5)) {speedY=0.1}) {samplerType=SAMPLERTYPE_Grayscale, texture=/FMCalibration/Textures/LowResBlurredNoise}
  n7 = ScalarParameter() {defaultValue=0.2, parameterName=Distortion MLT}
  n8 = ScalarParameter() {defaultValue=0.5, parameterName=SineSpeed}
  n9 = ScalarParameter() {defaultValue=0.5, parameterName=SineSpeed}
  n10 = MaterialFunctionCall(
   UVs=Add(A=Add(A=@n1, B=@n2), B=Multiply(A=@n6.R, B=@n7) {constB=0.2})
   Radius=LinearInterpolate(
    A=ScalarParameter() {defaultValue=0.4, parameterName=Radius_Start}
    B=ScalarParameter() {defaultValue=0.5, parameterName=Radius_Stop}
    Alpha=MaterialFunctionCall(Speed=@n8) {materialFunction=&DebugTimeSine}.Result
   ) {constA=0.4, constB=0.5}
   Density=LinearInterpolate(
    A=ScalarParameter() {defaultValue=2, parameterName=Density_Start}
    B=ScalarParameter() {defaultValue=2.333, parameterName=Density_Stop}
    Alpha=MaterialFunctionCall(Speed=@n9) {materialFunction=&DebugTimeSine}.Result
   ) {constA=0.4, constB=0.5}
  ) {materialFunction=&RadialGradientExponential}
  n11 = Multiply(
   A=@n10.RadialGradientExponential
   B=LinearInterpolate(
    A=ScalarParameter() {defaultValue=0.25, parameterName=Alpha_Start}
    B=ScalarParameter() {defaultValue=0.5, parameterName=Alpha_Stop}
    Alpha=ScalarParameter() {parameterName=Alpha_Anim}
   ) {constA=1.5953}
  )
  n12 = ScalarParameter() {defaultValue=1, parameterName=TaperGlow}
  Multiply(
   A=LinearInterpolate(
    A=VectorParameter() {parameterName=Color_A, defaultValue=<rgba 0 1 0 1>}.RGB
    B=VectorParameter() {parameterName=Color_B, defaultValue=<rgba 1 0 0 1>}.RGB
    Alpha=@n10.RadialGradientExponential
   )
   B=Saturate(None=Power(Base=@n11, Exp=@n12))
  )
 ]
}
