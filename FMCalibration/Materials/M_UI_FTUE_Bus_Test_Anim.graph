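# Translucent UI test material, apparently for an FTUE bus animation in FMCalibration, that builds two thin vertical box shapes.
# Bus Location and Arrow Location scalar parameters each lerp a horizontal UV offset, with fixed opposite vertical offsets separating the boxes.
# Each offset UV feeds MF_UI_Box, and a Max merges the two fills into one mask.
# The graph shows no connection from this mask or the Ball Color parameter to material outputs.
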Material /FMCalibration/Materials/M_UI_FTUE_Bus_Test_Anim
graph {
 properties {
  bAutomaticallySetUsageInEditor=false
  bUsedWithStaticMesh=false
  blendMode=BLEND_Translucent
  materialDomain=MD_UI
  usedShadingModels=Default Lit
 }
 nodes [
  n1 = ScalarParameter() {parameterName=Bus Location}
  n2 = Add(A=TextureCoordinate(), B=AppendVector(A=LinearInterpolate(Alpha=@n1) {constA=0.5, constAlpha=0.6, constB=-0.5}, B=Constant() {r=0.3}))
  n3 = Constant2Vector() {g=0.6, r=0.02}
  n4 = ScalarParameter() {parameterName=Arrow Location}
  n5 = Add(A=TextureCoordinate(), B=AppendVector(A=LinearInterpolate(Alpha=@n4) {constA=0.5, constAlpha=0.6, constB=-0.5}, B=Constant() {r=-0.3}))
  n6 = Constant2Vector() {g=0.6, r=0.02}
  Max(
   A=NamedRerouteDeclaration(None=MaterialFunctionCall(UVs=@n2, Size=@n3) {materialFunction=&MF_UI_Box}.Fill) {name=Ground_1, nodeColor=<rgba 0.9453 0 1 1>}
   B=NamedRerouteDeclaration(None=MaterialFunctionCall(UVs=@n5, Size=@n6) {materialFunction=&MF_UI_Box}.Fill) {name=Ground, nodeColor=<rgba 0.9453 0 1 1>}
  )
  NamedRerouteUsage()
  Reroute(None=VectorParameter() {parameterName=Ball Color, defaultValue=<rgba 1 1 1 1>}.RGB)
 ]
}
