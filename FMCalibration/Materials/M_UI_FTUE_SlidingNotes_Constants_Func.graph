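# Material function that supplies shared layout constants for the sliding-notes FTUE calibration UI, described as one sliding gem.
# It takes no inputs and outputs the near-to-far track perspective ratio, its inverse, the nearest track width, and gem width and height.
# The inverse ratio and gem width derive from the ratio and track width constants, so they follow edits to those values.
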MaterialFunction /FMCalibration/Materials/M_UI_FTUE_SlidingNotes_Constants_Func
graph {
 properties {description=One sliding gem}
 nodes [
  n1 = Constant() {r=1.6}
  FunctionOutput(None=@n1) {description="Ratio of near Track width at ProjectedTrack=1 over far Track width ProjectedTrack = 0", outputName=PerspectiveRatio}
  FunctionOutput(None=Divide(B=@n1)) {description=The inverse perspective ratio, outputName=InverseRatio, sortPriority=1}
  n2 = Constant() {r=0.9}
  FunctionOutput(None=Multiply(A=@n2) {constB=0.7}) {description=the width of the gem, outputName=GemWidth, sortPriority=3}
  FunctionOutput(None=Constant() {r=0.1}) {description=Height of the gem along the track, outputName=GemHeight, sortPriority=4}
  FunctionOutput(None=@n2) {description=The track width at its nearest, outputName=TrackWidth, sortPriority=2}
 ]
}
