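# Material function for the FTUE sliding-notes UI that maps a gem's linear track position (0 to 1) to a perspective-projected screen position.
# It gets PerspectiveRatio from a shared constants function and interpolates from it by track position to form a depth, whose reciprocal is the inverse distance.
# Outputs are a flipped, centered V coordinate, the projection's analytic derivative for scaling height, and the inverse distance for scaling other sizes.
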MaterialFunction /FMCalibration/Materials/M_UI_FTUE_SlidingNotes_Perspective_Func
graph {
 properties {description=One sliding gem}
 nodes [
  n1 = FunctionInput() {
   description="Takes a tarck position from 0 to 1 and outputs a projected track position from 0 to 1, but using perspective"
   inputName=TrackPosition
   inputType=FunctionInput_Scalar
  }
  n2 = MaterialFunctionCall() {materialFunction=&/M_UI_FTUE_SlidingNotes_Constants_Func}
  n3 = Divide(B=LinearInterpolate(A=@n2.PerspectiveRatio, Alpha=@n1) {desc=From Far to Near})
  FunctionOutput(None=Subtract(B=Multiply(A=@n1, B=@n3)) {constA=0.5, desc=Flip and Center}) {
   description="The projected height from .5 to -.5, going from top to bottom of the screen.  texture V of the material."
   outputName=ProjectedV
  }
  FunctionOutput(None=Multiply(A=@n2.PerspectiveRatio, B=Multiply(A=@n3, B=@n3))) {
   description="The first derivative of the Track to Projection, you can use this to set height by multiplying."
   outputName=Derivative
   sortPriority=1
  }
  FunctionOutput(None=@n3) {
   description="The inverse distance from the viewer.  You can multiply sizes by this to get projected orthogonal sizes at that point."
   outputName=InverseDistance
   sortPriority=2
  }
 ]
}
