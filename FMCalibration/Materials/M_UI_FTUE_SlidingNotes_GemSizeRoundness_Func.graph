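# Material function that places and shapes a gem on the perspective note track of the sliding-notes FTUE calibration UI.
# It feeds TrackPosition into the shared Perspective function and adds the projected V to the texture UV's vertical axis.
# Gem width scales with inverse distance and height with the projection derivative, using base dimensions from the shared Constants function.
# The scaled height also serves as the corner roundness output.
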MaterialFunction /FMCalibration/Materials/M_UI_FTUE_SlidingNotes_GemSizeRoundness_Func
graph {nodes [
  n1 = Constant()
  n2 = MaterialFunctionCall!(TrackPosition=FunctionInput() {description=Track Position from 0 to 1, inputName=TrackPosition, inputType=FunctionInput_Scalar}) {materialFunction=&/M_UI_FTUE_SlidingNotes_Perspective_Func}
  FunctionOutput(None=Add(A=FunctionInput() {description=Texture UV, inputName=TexUV, inputType=FunctionInput_Vector2}, B=AppendVector(A=@n1, B=@n2.ProjectedV))) {outputName=UV}
  n3 = MaterialFunctionCall() {materialFunction=&/M_UI_FTUE_SlidingNotes_Constants_Func}
  n4 = Multiply(A=@n2.Derivative, B=@n3.GemHeight)
  FunctionOutput(None=AppendVector(A=Multiply(A=@n2.InverseDistance, B=@n3.GemWidth) {constB=0.65}, B=@n4)) {outputName=Size, sortPriority=10}
  FunctionOutput(None=@n4) {outputName=CornerRoundness, sortPriority=20}
 ]}
