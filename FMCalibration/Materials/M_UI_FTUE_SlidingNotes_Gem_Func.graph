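# Material function that draws one sliding gem as a fill mask for the FTUE sliding-notes UI in FMCalibration.
# Track position (0 to 1) and texture UV go into the GemSizeRoundness helper function, whose output drives the UV, size and corner-roundness inputs of MF_UI_Box.
# The box's Fill is saturated and returned as the single Fill output.
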MaterialFunction /FMCalibration/Materials/M_UI_FTUE_SlidingNotes_Gem_Func
graph {
 properties {description=One sliding gem}
 nodes [
  n1 = MaterialFunctionCall!(
   TrackPosition=FunctionInput() {description=Track Position from 0 to 1, inputName=TrackPosition, inputType=FunctionInput_Scalar}
   TexUV=FunctionInput() {description=Texture UV, inputName=TexUV, inputType=FunctionInput_Vector2}
  ) {materialFunction=&/M_UI_FTUE_SlidingNotes_GemSizeRoundness_Func}
  FunctionOutput(None=Saturate(None=MaterialFunctionCall(UVs=@n1.UV, Size=@n1.UV, CornerRoundness=@n1.UV) {materialFunction=&MF_UI_Box}.Fill)) {outputName=Fill}
 ]
}
