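# Translucent UI material for an FTUE calibration animation: note gems slide along a track toward a catcher outline.
# A scalar parameter, or looping time under a Debug switch, drives track position, with three gems offset from it.
# Helper functions supply gem shape and perspective, while box functions draw track and catcher.
# Named reroutes carry masks, fades, colors and optional animated side arrows, antialiased from a distance-field texture.
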Material /FMCalibration/Materials/M_UI_FTUE_SlidingNotes_Anim
graph {
 properties {
  bAutomaticallySetUsageInEditor=false
  bUsedWithStaticMesh=false
  blendMode=BLEND_Translucent
  materialDomain=MD_UI
  usedShadingModels=Default Lit
 }
 nodes [
  n1 = NamedRerouteUsage()
  n2 = Constant() {r=0.45}
  ComponentMask(None=Sine(None=Multiply(A=@n1, B=@n2))) {g=true}
  n3 = Constant() {r=0.05}
  LinearInterpolate(
   A=Saturate(None=MaterialFunctionCall(Size=@n3, CornerRoundness=@n3) {materialFunction=&MF_UI_Box}.Fill)
   Alpha=MaterialFunctionCall(Size=@n3, CornerRoundness=@n3) {materialFunction=&MF_UI_Box}.Fill
  ) {desc="Whoa!  How is this doing anything????"}
  n4 = Sine(None=Multiply(A=NamedRerouteUsage()) {constB=3})
  LinearInterpolate(Alpha=@n4) {constA=0.6, constB=0.6}
  LinearInterpolate(Alpha=@n4) {constA=0.2, constB=0.2}
  n5 = Constant() {r=-0.4}
  LinearInterpolate(A=Constant() {r=-0.43}, B=@n5, Alpha=NamedRerouteUsage())
  n6 = Constant() {r=0.28}
  LinearInterpolate(A=Constant() {r=0.13}, B=@n6, Alpha=NamedRerouteUsage())
  Min(
   A=MaterialFunctionCall!(Float4=NamedRerouteUsage()) {materialFunction=/Game/Global/MaterialFunctions/MF_BreakComponents}.A
   B=Max(A=NamedRerouteUsage(), B=NamedRerouteUsage())
  )
  n7 = Add(A=NamedRerouteUsage(), B=NamedRerouteUsage())
  n8 = Subtract(A=@n7, B=NamedRerouteUsage())
  n9 = MaterialFunctionCall!(TrackPosition=@n8, TexUV=NamedRerouteUsage()) {materialFunction=&/M_UI_FTUE_SlidingNotes_Gem_Func}
  n10 = MaterialFunctionCall(TrackPosition=@n7, TexUV=NamedRerouteUsage()) {materialFunction=&M_UI_FTUE_SlidingNotes_Gem_Func}
  n11 = Subtract(A=@n8, B=NamedRerouteUsage())
  n12 = NamedRerouteUsage()
  n13 = LinearInterpolate(
   B=Max(
    A=Max(A=@n10.Fill, B=@n9.Fill) {constB=0}
    B=MaterialFunctionCall(TrackPosition=@n11, TexUV=@n12) {materialFunction=&M_UI_FTUE_SlidingNotes_Gem_Func}.Fill
   ) {constB=0}
   Alpha=NamedRerouteUsage()
  ) {constAlpha=1}
  n14 = LinearInterpolate(B=NamedRerouteUsage(), Alpha=NamedRerouteUsage())
  NamedRerouteDeclaration(None=LinearInterpolate(A=LinearInterpolate(A=@n13, Alpha=@n14), Alpha=NamedRerouteUsage())) {name=MainSHape, nodeColor=<rgba 1 0 0.2109 1>}
  n15 = StaticSwitchParameter(True=Frac(None=Time()), False=ScalarParameter() {defaultValue=0.1, parameterName=Ball Animation Parameter}) {parameterName=Debug}
  NamedRerouteDeclaration(None=Multiply(A=@n15, B=NamedRerouteUsage())) {name=GemPosition, nodeColor=<rgba 0.8281 1 0 1>}
  n16 = NamedRerouteUsage()
  n17 = Frac(None=NamedRerouteUsage())
  n18 = Multiply(A=Multiply(A=@n17) {constB=4}, B=Subtract(A=@n17))
  n19 = LinearInterpolate(Alpha=OneMinus(None=Add(A=@n18))) {constA=0.372, constAlpha=-1, constB=0.47}
  n20 = MaterialFunctionCall!(TrackPosition=NamedRerouteUsage()) {materialFunction=&/M_UI_FTUE_SlidingNotes_Perspective_Func}
  NamedRerouteDeclaration(None=Add(A=@n16, B=AppendVector(A=@n19, B=@n20.ProjectedV))) {name=PositionL, nodeColor=<rgba 1 0 0.5156 1>}
  n21 = Multiply(A=@n19) {constB=-1}
  NamedRerouteDeclaration(None=Add(A=@n16, B=AppendVector(A=@n21, B=@n20.ProjectedV))) {name=PositionR, nodeColor=<rgba 0 0.8359 1 1>}
  n22 = Constant()
  n23 = MaterialFunctionCall(UVs=Add(A=NamedRerouteUsage(), B=AppendVector(A=@n22, B=@n5)), Size=AppendVector(A=Constant() {r=1}, B=@n6), GlowMax=Constant() {r=0.12}) {materialFunction=&MF_UI_Box}
  NamedRerouteDeclaration(None=LinearInterpolate(B=Saturate(None=@n23.Fill), Alpha=NamedRerouteUsage())) {name=MaskLine, nodeColor=<rgba 0 1 0.625 1>}
  NamedRerouteDeclaration(None=@n15) {name=TrackPosition, nodeColor=<rgba 0.8281 1 0 1>}
  n24 = MaterialFunctionCall() {materialFunction=&/M_UI_FTUE_SlidingNotes_Constants_Func}
  n25 = LinearInterpolate(
   A=@n24.InverseRatio
   Alpha=MaterialFunctionCall() {materialFunction=&LinearGradient}.VGradient
  ) {constA=0.5}
  n26 = MaterialFunctionCall(
   UVs=NamedRerouteUsage()
   Size=AppendVector(A=Multiply(A=@n25, B=@n24.TrackWidth), B=Constant() {desc="> 1 to hide the box ends", r=1.2})
   OutlineThickness=Constant() {r=0.025}
  ) {materialFunction=&MF_UI_Box}
  NamedRerouteDeclaration(None=@n26.Fill) {name=TrackFilled, nodeColor=<rgba 0 0.0859 1 1>}
  NamedRerouteDeclaration(None=@n10.Fill) {name=Gem1, nodeColor=<rgba 0 0.2734 1 1>}
  n27 = VectorParameter() {parameterName=White, defaultValue=<rgba 1 1 1 1>}
  n28 = VectorParameter() {parameterName=PointerColorNormal, defaultValue=<rgba 1 0.0225 0.8046 1>}
  n29 = Max(A=NamedRerouteUsage(), B=NamedRerouteUsage())
  n30 = VectorParameter() {parameterName=Gem1ColorSelected, defaultValue=<rgba 0 0.0775 1 1>}
  n31 = LinearInterpolate(A=@n30.A, B=@n28.RGBA, Alpha=@n29)
  NamedRerouteDeclaration(None=LinearInterpolate(
    A=LinearInterpolate(A=@n27.RGBA, B=@n28.RGBA, Alpha=@n29)
    B=LinearInterpolate(A=@n31, B=@n30.RGBA, Alpha=@n30.A) {desc="This isn't quite right, but good enough for now."}
    Alpha=NamedRerouteUsage()
   )) {name=FinalColor, nodeColor=<rgba 0.5937 1 0 1>}
  NamedRerouteDeclaration(None=ScalarParameter() {desc=needs to be float, parameterName=b_HasArrows}) {name=HasArrows, nodeColor=<rgba 0.2891 1 0 1>}
  NamedRerouteDeclaration(None=TextureCoordinate()) {name=TexCoord, nodeColor=<rgba 1 0 0.2207 1>}
  n32 = NamedRerouteDeclaration(None=@n26.Outline) {name=TrackOutline, nodeColor=<rgba 0.4062 1 0 1>}
  n33 = NamedRerouteUsage()
  NamedRerouteDeclaration(None=LinearInterpolate(B=Saturate(None=@n32), Alpha=OneMinus(None=@n33))) {name=Grid, nodeColor=<rgba 0.2656 0 1 1>}
  n34 = MaterialFunctionCall!(TrackPosition=NamedRerouteUsage(), TexUV=NamedRerouteUsage()) {materialFunction=&/M_UI_FTUE_SlidingNotes_GemSizeRoundness_Func}
  n35 = Constant() {r=0.025}
  n36 = Multiply(A=@n35) {constB=0.5, desc="Outline is centered, so add .5"}
  n37 = Multiply(
   A=ScalarParameter() {defaultValue=0.025, desc=Gap, parameterName=CatcherGap}
   B=ScalarParameter() {defaultValue=1, parameterName=CatcherSize}
  )
  n38 = Add(A=@n35, B=ScalarParameter() {parameterName=AddOutlineThickness})
  n39 = MaterialFunctionCall(
   UVs=@n34.UV
   Size=Add(A=@n37, B=Add(A=@n34.Size, B=@n36))
   CornerRoundness=Add(A=@n37, B=Add(A=@n34.CornerRoundness, B=@n36))
   OutlineThickness=@n38
  ) {materialFunction=&MF_UI_Box}
  NamedRerouteDeclaration(None=LinearInterpolate(B=Saturate(None=@n39.Outline), Alpha=ScalarParameter() {defaultValue=0.8, desc="opacity, I assume", parameterName=OP})) {name=Catcher, nodeColor=<rgba 0.9453 1 0 1>}
  NamedRerouteDeclaration(None=Constant() {desc=Where in track space the catcher is, r=0.9}) {name=CatcherTrackPos, nodeColor=<rgba 0 0.6484 1 1>}
  n40 = ScalarParameter() {
   defaultValue=0.44
   desc="Spacing in track space between the gems.  You can make this as large as you want, and as small as .29 with our 3 gems"
   parameterName=GemSpacing
  }
  NamedRerouteDeclaration(None=@n40) {name=GemSpacing, nodeColor=<rgba 1 0 0.6562 1>}
  n41 = MaterialFunctionCall() {materialFunction=&LinearGradient}
  NamedRerouteDeclaration(None=Saturate(None=SmoothStep(Value=@n41.VGradient) {constMin=-0.02})) {name=FadeMaskGrid, nodeColor=<rgba 0 1 0.9063 1>}
  n42 = Multiply(A=NamedRerouteUsage(), B=Constant() {r=0.3})
  n43 = SmoothStep(Value=ComponentMask(None=Sine(None=@n42)) {g=true}) {constMax=1.2, constMin=-0.1}
  NamedRerouteDeclaration(None=LinearInterpolate(B=Saturate(None=@n43), Alpha=SmoothStep(Value=@n41.VGradient) {constMax=0.9, constMin=1})) {name=FadeMaskNotes, nodeColor=<rgba 0.3359 0 1 1>}
  n44 = ScalarParameter() {defaultValue=0.14, parameterName=ArrowSize}
  n45 = MaterialFunctionCall(UVs=NamedRerouteUsage(), Texture Scale=@n44) {materialFunction=&ScaleUVsByCenter}
  n46 = Constant() {r=0.25}
  n47 = MaterialFunctionCall(
   DistanceField=TextureSample(UVs=MaterialFunctionCall(UVs=@n45.UVs, "Rotation Angle (0-1)"=@n46) {materialFunction=&CustomRotator}.Rotated Values) {samplerSource=SSM_Clamp_WorldGroupSettings, texture=/FMCalibration/Textures/T_UI_Arrow}.RGB
   Threshold=Constant() {r=0.5}
  ) {materialFunction=&MF_UI_AntiAliasing}
  n48 = MaterialFunctionCall(UVs=NamedRerouteUsage(), Texture Scale=@n44) {materialFunction=&ScaleUVsByCenter}
  n49 = Constant() {r=-0.25}
  n50 = MaterialFunctionCall(
   DistanceField=TextureSample(UVs=MaterialFunctionCall(UVs=@n48.UVs, "Rotation Angle (0-1)"=@n49) {materialFunction=&CustomRotator}.Rotated Values) {samplerSource=SSM_Clamp_WorldGroupSettings, texture=&T_UI_Arrow}.RGB
   Threshold=Constant() {r=0.5}
  ) {materialFunction=&MF_UI_AntiAliasing}
  NamedRerouteDeclaration(None=LinearInterpolate(A=Constant(), B=LinearInterpolate(A=@n47.Outer, Alpha=@n50.Outer), Alpha=NamedRerouteUsage()) {constAlpha=0, constB=0}) {name=SideAnim, nodeColor=<rgba 1 0 0.2344 1>}
  n51 = Subtract(A=@n17) {constB=0.5}
  OneMinus(None=Abs(None=Multiply(A=@n51) {constB=2}))
  Sine(None=Multiply(B=@n17) {constA=0.5, constB=0})
  n52 = NamedRerouteUsage()
  StaticSwitchParameter(True=LinearInterpolate(A=@n38, B=@n35, Alpha=Sine(None=@n52)), False=@n38) {parameterName=LineAnimation}
 ]
}
