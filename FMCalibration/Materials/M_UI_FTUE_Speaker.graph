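# Translucent UI material for the calibration FTUE, showing either the HMX logo or a speaker graphic, with a Speaker_Color parameter.
# A static switch, defaulting to the logo, picks the logo texture through distance-field anti-aliasing or the speaker texture sharpened by SmoothStep.
# The OP scalar scales that mask through a lerp.
# An optional Fade switch caps it with an inverted, smoothstepped vertical gradient using Min.
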Material /FMCalibration/Materials/M_UI_FTUE_Speaker
graph {
 properties {
  bAutomaticallySetUsageInEditor=false
  bUsedWithStaticMesh=false
  blendMode=BLEND_Translucent
  materialDomain=MD_UI
  usedShadingModels=Default Lit
 }
 nodes [
  n1 = TextureSample() {texture=/FMCalibration/Textures/T_UI_FTUE_HMX_Logo}
  n2 = Constant() {r=0.9}
  n3 = Constant() {r=0.5}
  n4 = TextureSample() {texture=/FMCalibration/Textures/T_UI_FTUE_Speaker}
  n5 = LinearInterpolate(
   B=StaticSwitchParameter(
    True=MaterialFunctionCall(DistanceField=@n1.RGB, Threshold=@n2, Smoothing=@n3) {materialFunction=&MF_UI_AntiAliasing}.Outer
    False=SmoothStep(Value=@n4.RGB) {constMax=0.6, constMin=0.3}
   ) {defaultValue=true, parameterName=Logo / Speaker}
   Alpha=ScalarParameter() {defaultValue=1, parameterName=OP}
  )
  n6 = OneMinus(None=MaterialFunctionCall() {materialFunction=&LinearGradient}.VGradient)
  StaticSwitchParameter(True=Min(A=@n5, B=SmoothStep(Value=@n6) {constMax=0.6, constMin=0.35}), False=@n5) {parameterName=Fade}
  VectorParameter() {parameterName=Speaker_Color, defaultValue=<rgba 1 1 1 1>}
 ]
}
