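# UI material function that turns a signed distance field into animated, repeating wave bands that follow the shape's contours.
# It subtracts Motion from the SDF input, scales by Repetitions, then outputs the absolute cosine as a new SDF.
# Motion is a 0-1 value defaulting to the fractional part of scaled time, so the waves animate unless overridden.
# A custom-code hexagon SDF on centred UVs serves only as the input's preview.
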MaterialFunction /FMCalibration/Materials/MF_UI_MPT_Wave
graph {nodes [
  MaterialExpressionConstantBiasScale_0 = ConstantBiasScale(None=TextureCoordinate()) {bias=-0.5, scale=2}
  MaterialExpressionConstant_0 = Constant() {r=1}
  n1 = Custom!(uv=@MaterialExpressionConstantBiasScale_0, radius=@MaterialExpressionConstant_0) {
   code="float3 k = float3(-0.866025404,0.5,0.577350269);\r\nuv = abs(uv);\r\nuv -= 2.0*min(dot(k.xy,uv),0.0)*k.xy;\r\nuv -= float2(clamp(uv.x, -k.z*radius, k.z*radius), radius);\r\nreturn length(uv)*sign(uv.y);"
   description=Hexagon SDF
   inputs [
    {inputName=uv, input {expression=&self:MaterialExpressionConstantBiasScale_0}}
    {inputName=radius, input {expression=&self:MaterialExpressionConstant_0}}
   ]
  }
  n2 = Time()
  n3 = Constant() {desc=speed, r=0.5}
  n4 = FunctionInput(Preview=Frac(None=Multiply(A=@n2, B=@n3) {constB=0.5})) {bUsePreviewValueAsDefault=true, description="0-1 value, default is time + frac", inputName=Motion, inputType=FunctionInput_Scalar}
  n5 = Multiply(
   A=Subtract(A=FunctionInput(Preview=@n1) {inputName=SDF Input, inputType=FunctionInput_Scalar}, B=@n4)
   B=FunctionInput(Preview=Constant() {desc=repetitions, r=3}) {bUsePreviewValueAsDefault=true, inputName=Repetitions, inputType=FunctionInput_Scalar}
  )
  FunctionOutput(None=Abs(None=Cosine(None=@n5))) {outputName=SDF Output}
 ]}
