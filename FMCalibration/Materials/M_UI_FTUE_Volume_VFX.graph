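# Translucent UI material for an FTUE volume effect, with two animation styles chosen by a static switch and a Main Color parameter.
# V1 draws three rounded bars: a stepped gradient picks each column's offset, time shift and height range, and an eased sine animates the heights.
# V2 scrolls SDF ripple waves, trimmed by circle and radial-gradient masks.
# Animation Controller scales the motion of V1 and the visibility of V2.
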Material /FMCalibration/Materials/M_UI_FTUE_Volume_VFX
graph {
 properties {
  bAutomaticallySetUsageInEditor=false
  bUsedWithStaticMesh=false
  blendMode=BLEND_Translucent
  materialDomain=MD_UI
  usedShadingModels=Default Lit
 }
 nodes [
  Constant() {r=0.15}
  Constant() {r=0.7}
  n1 = Constant() {r=0.125}
  n2 = NamedRerouteUsage()
  NamedRerouteDeclaration(None=MaterialFunctionCall(UVs=NamedRerouteUsage(), Size=AppendVector(A=@n1, B=@n2), CornerRoundness=Constant() {r=0.125}) {materialFunction=&MF_UI_Box}.Fill) {name=Animation V1, nodeColor=<rgba 0.8984 1 0 1>}
  n3 = ScalarParameter() {defaultValue=3, desc=speed, parameterName=Ripple Frequency}
  n4 = ConstantBiasScale(None=Multiply(A=Time(), B=ScalarParameter() {defaultValue=0.5, parameterName=AnimationSpeed}))
  n5 = OneMinus(None=Frac(None=Multiply(A=@n4) {constB=0.25}))
  n6 = TextureCoordinate()
  n7 = Constant() {r=0.125}
  n8 = OneMinus(None=MaterialFunctionCall(
    UVs=MaterialFunctionCall(UVs=@n6, "Rotation Angle (0-1)"=@n7) {materialFunction=&CustomRotator}.Rotated Values
    Size=Constant() {r=1}
   ) {materialFunction=&MF_UI_Circle}.SDF)
  n9 = NamedRerouteDeclaration(None=Saturate(None=MaterialFunctionCall!(Repetitions=@n3, Motion=@n5, SDF Input=@n8) {materialFunction=&/MF_UI_MPT_Wave}.SDF Output)) {name=FracWaveLine, nodeColor=<rgba 1 0.6797 0 1>}
  n10 = LinearInterpolate(
   A=SmoothStep(Value=@n9) {constMax=0.15, constMin=0.1}
   B=SmoothStep(Value=@n9) {constMax=0.17, constMin=0.15}
   Alpha=Saturate(None=NamedRerouteUsage())
  )
  n11 = MaterialFunctionCall(Size=Constant() {r=0.55}, GlowMax=Constant() {r=0.35}) {materialFunction=&MF_UI_Circle}
  n12 = MaterialFunctionCall(Size=Constant() {r=0.2}, GlowMax=Constant() {r=0.35}) {materialFunction=&MF_UI_Circle}
  n13 = Min(A=Min(A=OneMinus(None=@n10), B=Saturate(None=@n11.Fill)), B=Saturate(None=OneMinus(None=@n12.Fill)))
  n14 = Saturate(None=MaterialFunctionCall() {materialFunction=&RadialGradientExponential}.RadialGradientExponential)
  NamedRerouteDeclaration(None=LinearInterpolate(B=LinearInterpolate(B=@n13, Alpha=@n14), Alpha=ScalarParameter() {defaultValue=1, parameterName=Animation Controller})) {name=Animation V2, nodeColor=<rgba 1 0 0.1641 1>}
  n15 = ScalarParameter() {parameterName=Time offset 1}
  n16 = ScalarParameter() {defaultValue=0.23, parameterName=Time offset 2}
  n17 = ScalarParameter() {defaultValue=0.57, parameterName=Time offset 3}
  n18 = NamedRerouteUsage()
  n19 = Multiply(
   A=Add(
    A=Time()
    B=MaterialFunctionCall(A=@n15, B=@n16, C=@n17, Alpha=@n18) {materialFunction=&Lerp_3Color}.Result
   )
   B=ScalarParameter() {defaultValue=0.5, parameterName=AnimationSpeed}
  )
  n20 = MaterialFunctionCall("Value (0-1)"=ConstantBiasScale(None=Sine(None=@n19))) {materialFunction=&MF_UI_EaseCurves}
  n21 = LinearInterpolate(Alpha=@n20.EaseInOutSine) {constA=0.125, constB=0.3}
  n22 = LinearInterpolate(Alpha=@n20.EaseInOutSine) {constA=0.4, constB=0.6}
  n23 = LinearInterpolate(Alpha=@n20.EaseInOutSine) {constA=0.7, constB=0.9}
  n24 = NamedRerouteUsage()
  NamedRerouteDeclaration(None=LinearInterpolate(
    A=Constant() {r=0.125}
    B=MaterialFunctionCall(A=@n21, B=@n22, C=@n23, Alpha=@n24) {materialFunction=&Lerp_3Color}.Result
    Alpha=ScalarParameter() {defaultValue=1, parameterName=Animation Controller}
   )) {name=Universal Anim, nodeColor=<rgba 1 0.1875 0 1>}
  n25 = MaterialFunctionCall() {materialFunction=&LinearGradient}
  n26 = Constant() {r=3}
  NamedRerouteDeclaration(None=Saturate(None=MaterialFunctionCall!("Value (0-1)"=@n25.UGradient, Steps=@n26) {materialFunction=/FortUILibrary/MaterialFunctions/Basics/MF_UI_StepRatio}.Stepped Gradient)) {name=3 Value Mask, nodeColor=<rgba 0 0.0625 1 1>}
  n27 = MaterialFunctionCall(
   A=Constant3Vector() {constant=<rgba 0.25 0 0 0>}
   B=Constant3Vector()
   C=Constant3Vector() {constant=<rgba -0.25 0 0 0>}
   Alpha=NamedRerouteUsage()
  ) {materialFunction=&Lerp_3Color}
  NamedRerouteDeclaration(None=Add(A=TextureCoordinate(), B=ComponentMask(None=@n27.Result) {g=true, r=true})) {name=Universal Position, nodeColor=<rgba 1 0 0.6328 1>}
  NamedRerouteDeclaration(None=MaterialFunctionCall(Radius=Constant() {r=0.4}, Density=Constant() {r=1}) {materialFunction=&RadialGradientExponential}.RadialGradientExponential) {name=Mask, nodeColor=<rgba 0 0.0859 1 1>}
  NamedRerouteUsage()
  StaticSwitchParameter(True=NamedRerouteUsage(), False=NamedRerouteUsage()) {defaultValue=true, parameterName=Animation Version V1/V2}
  VectorParameter() {parameterName=Main Color, defaultValue=<rgba 1 1 1 1>}
 ]
}
