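# Additive UI material that provides the pieces of a tinted circular glow with animated flicker and bob, named GlowTech.
# MF_UI_Circle builds the glow fill, which is multiplied by a parameterized flicker range driven by nested, saturated time sines, so negative half-cycles hold at minimum.
# A separate time sine scales a Y-axis bob offset added to world position.
# Glow Color is a standalone vector parameter for the tint.
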Material /FMCalibration/Materials/M_UI_GlowTech
graph {
 properties {bUsedWithStaticMesh=false, blendMode=BLEND_Additive, materialDomain=MD_UI, pixelDepthOffsetMode=PDOM_Legacy, usedShadingModels=Default Lit}
 nodes [
  n1 = Constant()
  n2 = AppendVector(A=@n1, B=ScalarParameter() {defaultValue=20, parameterName=Bob Height Offset})
  n3 = Multiply(A=Time(), B=ScalarParameter() {defaultValue=0.2, parameterName=BobRate})
  Add(A=LinearInterpolate(B=AppendVector(A=@n2, B=@n1), Alpha=Sine(None=@n3)) {constB=50}, B=WorldPosition().XYZ)
  n4 = Multiply(A=Time(), B=ScalarParameter() {defaultValue=0.2, parameterName=Flicker Speed})
  n5 = Sine(None=Multiply(A=Sine(None=@n4)) {constB=0.5})
  Multiply(
   A=MaterialFunctionCall(
    Size=ScalarParameter() {defaultValue=0.5, parameterName=Glow Size}
    GlowMin=ScalarParameter() {parameterName=Glow Min}
    GlowMax=ScalarParameter() {defaultValue=0.5, parameterName=Glow Max}
   ) {materialFunction=&MF_UI_Circle}.Fill
   B=LinearInterpolate(
    A=ScalarParameter() {defaultValue=0.5, parameterName=Flicker Range Min}
    B=ScalarParameter() {defaultValue=1, parameterName=Flicker Range Max}
    Alpha=Saturate(None=@n5)
   ) {constB=50}
  )
  VectorParameter() {parameterName=Glow Color, defaultValue=<rgba 0.5333 0.3231 0.0144 0.749>}
 ]
}
