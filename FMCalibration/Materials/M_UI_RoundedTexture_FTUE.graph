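# Masked UI material for an FMCalibration tutorial image (FTUE), pairing a fixed tutorial texture with a rounded-corner mask.
# Four Layout scalar parameters set the corner radii, packed by MakeFloat4 into MF_UI_CoreUI_RoundedMask.
# By default only the right-hand corners have a non-zero radius.
# The texture is a plain sample rather than a parameter, so instances can adjust corners but not swap the image.
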Material /FMCalibration/Materials/M_UI_RoundedTexture_FTUE
graph {
 properties {bAutomaticallySetUsageInEditor=false, bUsedWithStaticMesh=false, blendMode=BLEND_Masked, materialDomain=MD_UI, usedShadingModels=Default Lit}
 nodes [
  MaterialFunctionCall(Corner Radius=MaterialFunctionCall(
    X=ScalarParameter() {group=Layout, parameterName=TopLeft, sortPriority=10}
    Y=ScalarParameter() {group=Layout, parameterName=BottomLeft, sortPriority=13}
    Z=ScalarParameter() {defaultValue=0.25, group=Layout, parameterName=TopRight, sortPriority=11}
    A=ScalarParameter() {defaultValue=0.25, group=Layout, parameterName=BottomRight, sortPriority=12}
   ) {materialFunction=&MakeFloat4}.Result) {materialFunction=&MF_UI_CoreUI_RoundedMask}
  TextureSample() {texture=/FMCalibration/Textures/T_UI_LocalMP_Tutorial2}
 ]
}
