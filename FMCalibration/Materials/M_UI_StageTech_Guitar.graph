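# Translucent UI material for the StageTech guitar image in FMCalibration, drawing a parameterized texture, by default a SparksTutorial image.
# Texture UVs are scaled about the center, optionally mirrored horizontally by a static switch, then offset by Position X and Y and sampled with clamping.
# A Fade static switch, on by default, multiplies the texture alpha by an inverted, saturated SDF from an off-center circle to soften it.
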Material /FMCalibration/Materials/M_UI_StageTech_Guitar
graph {
 properties {
  bUsedWithStaticMesh=false
  blendMode=BLEND_Translucent
  materialDomain=MD_UI
  pixelDepthOffsetMode=PDOM_Legacy
  usedShadingModels=Default Lit
 }
 nodes [
  MaterialFunctionCall(ScalePivotPoint=Constant2Vector() {g=0.5, r=0.5}) {materialFunction=&MF_UI_UVScaling}
  n1 = MaterialFunctionCall(UVs=TextureCoordinate(), Texture Scale=ScalarParameter() {defaultValue=1.3, parameterName=Texture Scale}) {materialFunction=&ScaleUVsByCenter}
  n2 = AppendVector(A=StaticSwitchParameter(True=Constant() {r=-1}, False=Constant() {r=1}) {parameterName=Reflect Texture?}, B=Constant() {r=1})
  n3 = ScalarParameter() {parameterName=Position X}
  n4 = ScalarParameter() {defaultValue=-0.15, parameterName=Position Y}
  n5 = TextureSampleParameter2D(UVs=Add(
    A=MaterialFunctionCall(UVs=@n1.UVs, Texture Scale=@n2) {materialFunction=&ScaleUVsByCenter}.UVs
    B=AppendVector(A=@n3, B=@n4)
   ) {constB=-0.5}) {parameterName=Texture_Asset, samplerSource=SSM_Clamp_WorldGroupSettings, texture=/FMCalibration/Textures/T_UI_SparksTutorial_MRKT}
  n6 = TextureCoordinate()
  n7 = Constant2Vector() {g=0.25, r=-0.1}
  n8 = Multiply(A=MaterialFunctionCall(UVs=Add(A=@n6, B=@n7), Size=Constant() {r=1.6}) {materialFunction=&MF_UI_Circle}.SDF) {constB=-1}
  StaticSwitchParameter(True=Multiply(A=@n5.A, B=Saturate(None=@n8)), False=@n5.A) {defaultValue=true, parameterName=Fade}
 ]
}
