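# Translucent UI material for an animated atmospheric backdrop of stars and drifting dust, named for the FTUE screens.
# Panning starfield and warped star-noise textures blend through a cloud texture, then add to two counter-scrolling dust layers on warped UVs, with opacity parameters for each.
# The star blend interpolates between two color parameters.
# The combined layers are masked by an eased radial gradient times a linear gradient, with a separate vertical-gradient fade branch.
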Material /FMCalibration/Materials/M_Ui_FTUE_RadiatingAtmospherics
graph {
 properties {
  bUsedWithStaticMesh=false
  blendMode=BLEND_Translucent
  materialDomain=MD_UI
  pixelDepthOffsetMode=PDOM_Legacy
  usedShadingModels=Default Lit
 }
 nodes [
  n1 = Time()
  n2 = Multiply(A=TextureCoordinate(), B=AppendVector(A=Constant() {r=1.2}, B=Constant() {r=1})) {constB=1.5}
  n3 = Time()
  n4 = Panner(
   Coordinate=MaterialFunctionCall!(UV=@n2, Warp Center, Warp Intensity=Constant() {r=-0.15}, Warp Radius=Constant() {r=-0.5}) {materialFunction=/FortUILibrary/MaterialFunctions/Effects/MF_UI_Warp}.Warped UV
   Time=@n3
  ) {speedY=0.02}
  n5 = LinearInterpolate(
   A=TextureSample(UVs=Panner(Time=@n1) {speedY=0.03}) {texture=/Game/Global/Textures/Default/PointSplatter/T_UI_Starfield}.RGB
   B=TextureSample(UVs=Multiply(A=@n4) {constB=2.6}) {texture=/Game/Effects/Fort_Effects/Gameplay/Butterfly/T_noise_stars}.RGB
   Alpha=TextureSample() {texture=/Game/Packages/Fortress_Transylvania/Haunted_Forest/Swamps/Textures/clouds_dust_detail_01}.RGB
  )
  n6 = MaterialFunctionCall(
   UV=@n2
   Warp Center=AppendVector(A=Constant() {r=0.5}, B=Constant() {r=1})
   Warp Intensity=Constant() {r=0.25}
   Warp Radius=Constant() {r=0.35}
  ) {materialFunction=&MF_UI_Warp}
  n7 = Panner(Coordinate=Multiply(A=@n6.Warped UV) {constB=5}, Time=@n3) {speedX=0.007, speedY=0.02}
  n8 = Panner(Coordinate=Multiply(A=@n6.Warped UV) {constB=2}, Time=@n3) {speedX=-0.008, speedY=-0.02}
  n9 = Add(
   A=Multiply(A=TextureSample(UVs=@n7) {texture=&clouds_dust_detail_01}.RGB) {constB=0.6}
   B=Multiply(A=TextureSample(UVs=@n8) {samplerType=SAMPLERTYPE_Grayscale, texture=/FortUILibrary/UIKit/Buttons/Textures/LowResBlurredNoise}.RGB) {constB=0.3}
  )
  n10 = Add(
   A=LinearInterpolate(B=@n5, Alpha=ScalarParameter() {defaultValue=1, parameterName=DrstOp})
   B=LinearInterpolate(B=Saturate(None=@n9), Alpha=ScalarParameter() {defaultValue=0.5, parameterName=BG_DustOP})
  )
  LinearInterpolate(
   B=@n10
   Alpha=SmoothStep(Value=MaterialFunctionCall() {materialFunction=&LinearGradient}.VGradient) {constMin=0.4}
  )
  LinearInterpolate(
   A=VectorParameter() {parameterName=BgDustColor, defaultValue=<rgba 0.0542 0.0759 1 1>}.RGB
   B=VectorParameter() {parameterName=Atmospherics Color, defaultValue=<rgba 1 1 1 1>}.RGB
   Alpha=@n5
  )
  n11 = Constant()
  n12 = Constant()
  n13 = MaterialFunctionCall(UV=Add(A=TextureCoordinate(), B=AppendVector(A=@n11, B=@n12)), GradientEnd=Constant() {r=0.5}) {materialFunction=&MF_UI_RadialGradient}
  n14 = Constant() {r=0.25}
  n15 = Constant() {r=0.45}
  n16 = Constant() {r=0.25}
  Multiply(
   A=@n10
   B=Multiply(
    A=MaterialFunctionCall("Value (0-1)"=@n13.Smooth) {materialFunction=&MF_UI_EaseCurves}.EaseOutSine
    B=MaterialFunctionCall(GradientStart=@n14, GradientEnd=@n15, RotationAngle=@n16) {materialFunction=&MF_UI_LinearGradient}.Result
   )
  )
 ]
}
