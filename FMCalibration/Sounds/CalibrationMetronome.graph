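# CalibrationMetronome is a looping stereo MetaSound source that plays a click in time with a Harmonix metronome, apparently for calibration in FMCalibration.
# On play, a trigger-to-transport node starts the metronome, which follows the Tempo and Speed inputs.
# Its MIDI clock drives a subdivision trigger that fires the click wave player, and is also exposed as an output.
# Both audio channels are scaled by twice the Gain input.
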MetaSoundSource /FMCalibration/Sounds/CalibrationMetronome
inputs [
 Gain {type=Float}
 Speed {type=Float}
 Tempo {type=Float}
 UE.Source.OnPlay {type=Trigger}
]
interfaces [UE.OutputFormat.Stereo:1.0, UE.Source:1.1]
outputs [
 Midi Clock {type=MidiClock}
 "UE.OutputFormat.Stereo.Audio:0" {type=Audio}
 "UE.OutputFormat.Stereo.Audio:1" {type=Audio}
]
pages [Default {
  input_defaults {Gain=1, Speed=1, Tempo=80, UE.Source.OnPlay=null}
  nodes [
   n1 = UE.Multiply:Float(PrimaryOperand=^Gain) {also {defaults {AdditionalOperands=2}}}
   HarmonixNodesMetronome_0 = HarmonixNodes.Metronome(Speed=^Speed, Tempo=^Tempo, Transport=HarmonixNodes.TriggerToTransport(Play=^UE.Source.OnPlay).Transport)
   WavePlayer_0 = UE.Wave Player:Stereo(Play=HarmonixNodes.MidiClockSubdivisionTrigger(Midi Clock=@HarmonixNodesMetronome_0.Midi Clock) {also {defaults {"Base Offset Size "=0}}}.Trigger Out) {also {defaults {Wave Asset=&/SparksCalibrationClick}}}
   Multiply_0 = UE.Multiply:Audio by Float(AdditionalOperands=@n1.Out, PrimaryOperand=@WavePlayer_0.Out Right)
   Multiply_1 = UE.Multiply:Audio by Float(AdditionalOperands=@n1.Out, PrimaryOperand=@WavePlayer_0.Out Left)
  ]
  links [
   {dst="@Midi Clock", src=HarmonixNodesMetronome_0.Midi Clock}
   {dst="@UE.OutputFormat.Stereo.Audio:1", src=Multiply_0.Out}
   {dst="@UE.OutputFormat.Stereo.Audio:0", src=Multiply_1.Out}
  ]
 }]
properties {
 assetImportData=&self:AssetImportData
 bLooping=true
 duration=10000
 numChannels=2
 outputFormat=Stereo
 sampleRate=48000
 blockRateOverride {default=100}
}
