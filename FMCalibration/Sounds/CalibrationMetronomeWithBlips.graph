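# Mono looping MetaSound source for calibration: a Harmonix metronome, started on play and set by Tempo and Speed, outputs its MIDI clock.
# The clock's subdivision trigger fires a snare, mixed with a blip player under Gain-scaled levels.
# A four-step counter selects each blip's wave and volume from arrays, but its subdivision trigger has no clock input wired.
# The ChirpVolume input is unconnected.
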MetaSoundSource /FMCalibration/Sounds/CalibrationMetronomeWithBlips
inputs [
 ChirpVolume {type=Float}
 Gain {type=Float}
 Speed {type=Float}
 Tempo {type=Float}
 UE.Source.OnPlay {type=Trigger}
]
interfaces [UE.OutputFormat.Mono:1.0, UE.Source:1.1]
outputs [
 Midi Clock {type=MidiClock}
 "UE.OutputFormat.Mono.Audio:0" {type=Audio}
]
pages [Default {
  input_defaults {ChirpVolume=0.504, Gain=1, Speed=1, Tempo=60, UE.Source.OnPlay=null}
  nodes [
   TriggerCounter_0 = UE.Trigger Counter(In=HarmonixNodes.MidiClockSubdivisionTrigger() {also {defaults {Base Grid Size=3, "Base Offset Size "=4, Offset Multiplier=1}}}.Trigger Out) {also {defaults {Reset Count=4}}}
   n1 = Convert.Int32:Float(Float=@TriggerCounter_0.Value)
   n2 = UE.Reroute(Value=@n1.Int32)
   n3 = VariableAccessor.Float:Array(Variable=InitVariable.Float:Array().Variable)
   n4 = UE.Reroute(Value=@TriggerCounter_0.On Trigger)
   n5 = InitVariable.WaveAsset:Array()
   WavePlayer_0 = UE.Wave Player:Mono(
    Wave Asset=Array.Get:WaveAsset:Array(Index=@n1.Int32, Trigger=@n4.Value, Array=VariableAccessor.WaveAsset:Array(Variable=@n5.Variable).Value).Element
    Play=@n4.Value
   )
   HarmonixNodesMetronome_0 = HarmonixNodes.Metronome(Speed=^Speed, Tempo=^Tempo, Transport=HarmonixNodes.TriggerToTransport(Play=^UE.Source.OnPlay).Transport)
   WavePlayer_1 = UE.Wave Player:Mono(Play=UE.Reroute(Value=HarmonixNodes.MidiClockSubdivisionTrigger(Midi Clock=@HarmonixNodesMetronome_0.Midi Clock) {also {defaults {"Base Offset Size "=0}}}.Trigger Out).Value) {also {defaults {Wave Asset=&/SparksCalibrationSnareGeneric}}}
   AudioMixerAudioMixerMono2_0 = "AudioMixer.Audio Mixer (Mono, 2)"(
    In 1=UE.Multiply:Audio by Float(
     AdditionalOperands=Array.Get:Float:Array(Index=@n2.Value, Array=@n3.Value, Trigger=@WavePlayer_0.On Play).Element
     PrimaryOperand=UE.Reroute(Value=@WavePlayer_0.Out Mono).Value
    ).Out
    Gain 0=UE.Multiply:Float(PrimaryOperand=^Gain) {also {defaults {AdditionalOperands=2}}}.Out
    Gain 1=UE.Multiply:Float(PrimaryOperand=^Gain).Out
    In 0=@WavePlayer_1.Out Mono
   )
  ]
  links [{dst="@UE.OutputFormat.Mono.Audio:0", src=AudioMixerAudioMixerMono2_0.Out}, {dst="@Midi Clock", src=HarmonixNodesMetronome_0.Midi Clock}]
 }]
properties {
 assetImportData=&self:AssetImportData
 bLooping=true
 duration=10000
 sampleRate=48000
 blockRateOverride {default=100}
}
