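# Looping mono MetaSound source that synthesizes a metronome click, apparently for the FMCalibration calibrometer.
# OnPlay starts a Harmonix metronome driven by the Tempo and Speed inputs, whose MIDI clock is also exposed as an output.
# Clock subdivision triggers fire a short, sharp AD envelope that shapes a saw oscillator into clicks, scaled by Gain, with no samples used.
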MetaSoundSource /FMCalibration/Sounds/CalibrometerMetronome
inputs [
 Gain {type=Float}
 Speed {type=Float}
 Tempo {type=Float}
 UE.Source.OnPlay {type=Trigger}
]
interfaces [UE.OutputFormat.Mono:1.0, UE.Source:1.1]
outputs [
 Midi Clock {type=MidiClock}
 "UE.OutputFormat.Mono.Audio:0" {type=Audio}
]
pages [Default {
  input_defaults {Gain=1, Speed=1, Tempo=29.793, UE.Source.OnPlay=null}
  nodes [
   HarmonixNodesMetronome_0 = HarmonixNodes.Metronome(Speed=^Speed, Tempo=^Tempo, Transport=HarmonixNodes.TriggerToTransport(Play=^UE.Source.OnPlay).Transport)
   ADEnvelopeADEnvelope_0 = AD Envelope.AD Envelope:Audio(Trigger=HarmonixNodes.MidiClockSubdivisionTrigger(Midi Clock=@HarmonixNodesMetronome_0.Midi Clock) {also {defaults {"Base Offset Size "=0}}}.Trigger Out) {also {defaults {Attack Time=0.0001, Decay Curve=100, Decay Time=0.06}}}
   Multiply_1 = UE.Multiply:Audio by Float(
    AdditionalOperands=^Gain
    PrimaryOperand=UE.Multiply:Audio(PrimaryOperand=@ADEnvelopeADEnvelope_0.Out Envelope, AdditionalOperands=UE.Saw:Audio() {also {defaults {Frequency=1000}}}.Audio).Out
   )
  ]
  links [{dst="@Midi Clock", src=HarmonixNodesMetronome_0.Midi Clock}, {dst="@UE.OutputFormat.Mono.Audio:0", src=Multiply_1.Out}]
 }]
properties {
 assetImportData=&self:AssetImportData
 bLooping=true
 duration=10000
 sampleRate=48000
 blockRateOverride {default=100}
}
