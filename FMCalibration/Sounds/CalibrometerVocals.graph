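# Looping mono MetaSound source for the Calibrometer vocals, a clock-driven stepped saw tone that also outputs its MIDI clock.
# On play, a Harmonix metronome runs at the input tempo and speed, feeding a subdivision trigger.
# A trigger counter with a reset count of two indexes an integer note array on each subdivision.
# The selected MIDI note sets the saw oscillator frequency, and the Gain input scales the output.
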MetaSoundSource /FMCalibration/Sounds/CalibrometerVocals
inputs [
 Gain {type=Float}
 Speed {type=Float}
 Tempo {type=Float}
 UE.Source.OnPlay {type=Trigger}
]
interfaces [UE.OutputFormat.Mono:1.0, UE.Source:1.1]
outputs [
 Midi Clock {type=MidiClock}
 "UE.OutputFormat.Mono.Audio:0" {type=Audio}
]
pages [Default {
  input_defaults {Gain=1, Speed=1, Tempo=60, UE.Source.OnPlay=null}
  nodes [
   HarmonixNodesMetronome_0 = HarmonixNodes.Metronome(Speed=^Speed, Tempo=^Tempo, Transport=HarmonixNodes.TriggerToTransport(Play=^UE.Source.OnPlay).Transport)
   TriggerCounter_0 = UE.Trigger Counter(In=HarmonixNodes.MidiClockSubdivisionTrigger(Midi Clock=@HarmonixNodesMetronome_0.Midi Clock) {also {defaults {Base Grid Size=4}}}.Trigger Out) {also {defaults {Reset Count=2}}}
   n1 = InitVariable.Int32:Array()
   MIDIToFrequency_0 = UE.MIDI To Frequency:Int32(MIDI In=Array.Get:Int32:Array(
     Index=Convert.Int32:Float(Float=@TriggerCounter_0.Value).Int32
     Trigger=@TriggerCounter_0.On Trigger
     Array=VariableAccessor.Int32:Array(Variable=@n1.Variable).Value
    ).Element)
   Multiply_0 = UE.Multiply:Audio by Float(AdditionalOperands=^Gain, PrimaryOperand=UE.Saw:Audio(Frequency=@MIDIToFrequency_0.Out Frequency).Audio)
  ]
  links [{dst="@Midi Clock", src=HarmonixNodesMetronome_0.Midi Clock}, {dst="@UE.OutputFormat.Mono.Audio:0", src=Multiply_0.Out}]
 }]
properties {
 assetImportData=&self:AssetImportData
 bLooping=true
 duration=10000
 sampleRate=48000
 blockRateOverride {default=100}
}
